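# Material function for the Clyde LemonCart set that makes an object float through World Position Offset and outputs an animated emissive glow.
# Sine-driven X, Y and Z rotations about the object pivot, chained through named reroutes, add to a per-axis sine bob.
# The glow multiplies Fresnel by time-scrolling sine bands in a sphere mask offset from the pivot, then passes through MF_FX_EmissiveAdjustments and a color tint.
# Scalar parameters blend each output in, with WPO off by default.
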MaterialFunction /Clyde/Environments/Sets/LemonCart/MaterialFunctions/MF_Clyde_FloatWPO
graph {nodes [
  n1 = Constant3Vector() {constant=<rgba 0 0 1 0>}
  n2 = Sine(None=Multiply(A=Time()) {constB=0.02})
  n3 = LinearInterpolate(Alpha=Divide(A=Add(A=@n2))) {constA=-1}
  n4 = ScalarParameter() {defaultValue=0.01, parameterName=Z_Roate_Weight}
  n5 = NamedRerouteUsage()
  n6 = WorldPosition()
  n7 = Add(
   A=RotateAboutAxis(
    NormalizedRotationAxis=Normalize(VectorInput=@n1)
    RotationAngle=Multiply(A=@n3, B=@n4) {constB=0}
    PivotPoint=MaterialFunctionCall() {materialFunction=&ObjectPivotPoint}.Object Pivot Location
    Position=Add(A=@n5, B=@n6.XYZ)
   )
   B=Add(A=NamedRerouteUsage(), B=NamedRerouteUsage())
  )
  n8 = Sine(None=Multiply(A=Time()) {constB=0.3})
  n9 = Sine(None=Multiply(A=Time()) {constB=0.23})
  n10 = Sine(None=Multiply(A=Time()) {constB=0.23})
  n11 = Add(B=MaterialFunctionCall(X=Multiply(A=@n8) {constB=3}, Y=Multiply(A=@n9) {constB=10}, Z=Multiply(A=@n10) {constB=18}) {materialFunction=&MakeFloat3}.Result)
  FunctionOutput(None=LinearInterpolate(B=Add(A=@n7, B=@n11), Alpha=ScalarParameter() {parameterName=UseWPO}))
  n12 = Fresnel(ExponentIn=Constant() {r=3}, BaseReflectFractionIn=Constant() {r=0.004})
  n13 = SphereMask(
   A=WorldPosition().XYZ
   B=Add(
    A=MaterialFunctionCall() {materialFunction=&ObjectPivotPoint}.Object Pivot Location
    B=VectorParameter() {parameterName=Parama, defaultValue=<rgba 0 -200 60 1>}.RGB
   )
   Radius=Constant() {r=512}
  ) {hardnessPercent=0}
  n14 = Constant() {r=2}
  n15 = Time()
  n16 = Constant() {r=0.56}
  n17 = Sine(None=Add(A=Multiply(A=@n13, B=@n14) {constB=2}, B=Multiply(A=@n15, B=@n16)))
  n18 = Multiply(A=Multiply(A=Add(A=@n17)) {constB=0.5}, B=OneMinus(None=Power(Base=@n13) {constExponent=3}))
  n19 = Power(Base=@n13)
  n20 = MaterialFunctionCall(Emissive=Multiply(A=Power(Base=@n12), B=Multiply(A=@n18, B=@n19))) {materialFunction=&MF_FX_EmissiveAdjustments}
  n21 = VectorParameter() {parameterName=EmissiveColor_dj, defaultValue=<rgba 0 10.1228 20 1>}
  FunctionOutput(None=LinearInterpolate(B=Multiply(A=@n20.Result, B=@n21.RGB), Alpha=ScalarParameter() {defaultValue=1, parameterName=UseEMissive})) {outputName=Emissive}
  n22 = Constant3Vector() {constant=<rgba 0 1 0 0>}
  n23 = Sine(None=Multiply(A=Time()) {constB=0.1})
  n24 = LinearInterpolate(Alpha=Divide(A=Add(A=@n23))) {constA=-1}
  n25 = ScalarParameter() {defaultValue=0.01, parameterName=Y_Roate_Weight}
  NamedRerouteDeclaration(None=RotateAboutAxis(
    NormalizedRotationAxis=Normalize(VectorInput=@n22)
    RotationAngle=Multiply(A=@n24, B=@n25) {constB=0}
    PivotPoint=MaterialFunctionCall() {materialFunction=&ObjectPivotPoint}.Object Pivot Location
    Position=WorldPosition().XYZ
   )) {name=Y_Roate, nodeColor=<rgba 0.1484 0 1 1>}
  n26 = Constant3Vector() {constant=<rgba 1 0 0 0>}
  n27 = Sine(None=Multiply(A=Time()) {constB=0.1})
  n28 = LinearInterpolate(Alpha=Divide(A=Add(A=@n27))) {constA=-1}
  n29 = ScalarParameter() {defaultValue=0.005, parameterName=X_Roate_Weight}
  n30 = NamedRerouteUsage()
  n31 = WorldPosition()
  NamedRerouteDeclaration(None=RotateAboutAxis(
    NormalizedRotationAxis=Normalize(VectorInput=@n26)
    RotationAngle=Multiply(A=@n28, B=@n29) {constB=0}
    PivotPoint=MaterialFunctionCall() {materialFunction=&ObjectPivotPoint}.Object Pivot Location
    Position=Add(A=@n30, B=@n31.XYZ)
   )) {name=X_Roate, nodeColor=<rgba 0.1719 1 0 1>}
  ObjectOrientation()
 ]}
