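# Material function that outputs an animated, soft circular smoke mask, named for the LemonCart subscreen.
# Two red-channel samples of an inky smoke tile pan diagonally at different scales and speeds.
# Their remapped values warp the UVs of an exponential radial gradient by EdgeWarp, which also offsets its center.
# The gradient masks the saturated product of both samples, adds a faint Fill, then scales by a levels boost.
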MaterialFunction /Clyde/Environments/Sets/LemonCart/MaterialFunctions/MF_LemonCart_Subscreen_Smoke
graph {nodes [
  n1 = TextureCoordinate()
  n2 = Time()
  n3 = Constant4Vector() {constant=<rgba -0.02 0.01 -0.05 0.04>}
  n4 = TextureObject() {texture=&T_Inky_Smoke_Tile}
  n5 = TextureSample(
   UVs=Panner(
    Coordinate=Multiply(A=@n1) {constB=0.8}
    Time=@n2
    Speed=MaterialFunctionCall(X=@n3.R, Y=@n3.R) {materialFunction=&MakeFloat2}.Result
   )
   Tex=@n4
  ) {texture=&T_Inky_Smoke_Tile}
  n6 = TextureSample(
   UVs=Panner(
    Coordinate=Multiply(A=@n1) {constB=1.3}
    Time=@n2
    Speed=MaterialFunctionCall(X=@n3.B, Y=@n3.B) {materialFunction=&MakeFloat2}.Result
   )
   Tex=@n4
  ) {texture=&T_Inky_Smoke_Tile}
  n7 = Multiply(A=Subtract(A=MaterialFunctionCall(X=@n5.R, Y=@n6.R) {materialFunction=&MakeFloat2}.Result) {constB=0.5}) {constB=2}
  n8 = ScalarParameter() {defaultValue=0.1, parameterName=EdgeWarp}
  n9 = MaterialFunctionCall(
   UVs=Add(A=TextureCoordinate(), B=Multiply(A=@n7, B=@n8) {constB=0.1})
   CenterPosition=Subtract(B=Divide(A=@n8)) {constA=0.5}
   Radius=ScalarParameter() {defaultValue=0.4, parameterName=Radius}
  ) {materialFunction=&RadialGradientExponential}
  n10 = Multiply(A=@n9.RadialGradientExponential, B=ScalarParameter() {defaultValue=0.01, parameterName=Fill})
  n11 = Multiply(A=@n5.R, B=@n6.R)
  n12 = Multiply(A=@n9.RadialGradientExponential, B=Saturate(None=Multiply(A=@n11) {constB=2}))
  FunctionOutput(None=Multiply(A=Add(A=@n10, B=@n12), B=ScalarParameter() {defaultValue=2, parameterName=MaskLevelsBoostMulti}))
 ]}
