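# Material function for the LemonCart environment set that builds a custom MIP level from camera-to-actor distance, though its output is disconnected from that math.
# Distance past an offset is scaled, raised to a slight power, clamped by the rounded, capped NumberOfMips input, then floored.
# The floored result feeds nothing: the function output is wired to a zero constant, so it always returns 0 as saved.
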MaterialFunction /Clyde/Environments/Sets/LemonCart/MaterialFunctions/MF_LemonCart_CustomScreenMIPOut
graph {nodes [
  n1 = CameraPositionWS()
  n2 = ActorPositionWS()
  n3 = Divide(A=Subtract(A=Distance(A=@n1, B=@n2)) {constB=500}) {constB=500}
  n4 = Round(None=FunctionInput(Preview=Constant() {r=11}) {bUsePreviewValueAsDefault=true, inputName=NumberOfMips, inputType=FunctionInput_Scalar})
  Floor(None=Clamp(None=Power(Base=@n3) {constExponent=1.2}, Max=Clamp(None=@n4) {maxDefault=10}))
  FunctionOutput(None=Constant())
 ]}
