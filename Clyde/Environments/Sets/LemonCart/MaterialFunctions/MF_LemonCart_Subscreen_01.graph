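# Material function for a LemonCart sub-screen that draws a tinted icon drifting around like a screensaver, output as a color and an alpha.
# Time, offset by a Seed input and scaled by speed, drives nested looping three-point lerps for X/Y motion within edge padding.
# The moved UVs are center-scaled, jittered and bump-offset to sample the screen texture.
# Red channel times the 0-1 UV mask gives alpha; color is a parameter.
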MaterialFunction /Clyde/Environments/Sets/LemonCart/MaterialFunctions/MF_LemonCart_Subscreen_01
graph {nodes [
  FunctionOutput(None=VectorParameter() {parameterName=SubScreen01.Color.Icon, defaultValue=<rgba 1 0.8013 0 1>}.RGB)
  n1 = TextureCoordinate()
  n2 = MaterialFunctionCall(
   X=ScalarParameter() {defaultValue=0.38, parameterName=SubScreen01.EdgePadding.X}
   Y=ScalarParameter() {defaultValue=0.32, parameterName=SubScreen01.EdgePadding.Y}
  ) {materialFunction=&MakeFloat2}
  n3 = Constant() {r=1}
  n4 = Constant() {r=-1}
  n5 = Constant() {r=1}
  n6 = FunctionInput() {bUsePreviewValueAsDefault=true, inputName=Seed, inputType=FunctionInput_Scalar, sortPriority=4}
  n7 = Multiply(A=Add(A=Time(), B=Multiply(A=@n6) {constB=5}), B=ScalarParameter() {defaultValue=0.5, parameterName=SubScreen01.Speed}) {constB=3}
  n8 = Frac(None=Multiply(A=Add(A=@n7) {constB=0.21}) {constB=0.045})
  n9 = Multiply(A=Add(
    A=MaterialFunctionCall(A=@n4, B=@n5, C=@n4, Alpha=@n8) {materialFunction=&Lerp_3Color}.Result
    B=@n7
   )) {constB=0.2}
  n10 = Constant() {r=1}
  n11 = Constant() {r=-1}
  n12 = Constant() {r=1}
  n13 = Frac(None=Multiply(A=@n7) {constB=0.033})
  n14 = Multiply(A=Add(
    A=@n7
    B=MaterialFunctionCall(A=@n11, B=@n12, C=@n11, Alpha=@n13) {materialFunction=&Lerp_3Color}.Result
   ) {constB=0.33}) {constB=0.1}
  n15 = MaterialFunctionCall(
   X=MaterialFunctionCall(A=@n3, B=Constant() {r=-1}, C=@n3, Alpha=Frac(None=@n9)) {materialFunction=&Lerp_3Color}.Result
   Y=MaterialFunctionCall(A=@n10, B=Constant() {r=-1}, C=@n10, Alpha=Frac(None=@n14)) {materialFunction=&Lerp_3Color}.Result
  ) {materialFunction=&MakeFloat2}
  n16 = MaterialFunctionCall(
   UVs=Add(
    A=FunctionInput(Preview=@n1) {bUsePreviewValueAsDefault=true, inputName=UVs, inputType=FunctionInput_Vector2}
    B=Multiply(A=@n2.Result, B=@n15.Result)
   )
   Texture Scale=ScalarParameter() {defaultValue=0.5, parameterName=SubScreen01.Icon.Scale}
  ) {materialFunction=&ScaleUVsByCenter}
  n17 = FunctionInput() {bUsePreviewValueAsDefault=true, inputName=Height Bump Offset, inputType=FunctionInput_Scalar, sortPriority=1}
  n18 = Add(
   A=Add(A=@n16.UVs, B=FunctionInput() {bUsePreviewValueAsDefault=true, inputName=Vertical Jitter, inputType=FunctionInput_Vector2, sortPriority=2})
   B=MaterialFunctionCall(X=Multiply(A=@n17) {constB=-0.1}, Y=Constant()) {materialFunction=&MakeFloat2}.Result
  )
  n19 = TextureSample(UVs=BumpOffset(Coordinate=@n18, Height=@n17)) {texture=/Clyde/Environments/Sets/LemonCart/Textures/T_LemonCart_Screen_01}
  FunctionOutput(None=Saturate(None=Multiply(A=@n19.R, B=@n16.0-1 mask))) {outputName=Alpha, sortPriority=1}
  n20 = Sine(None=Multiply(A=Time()) {constB=0.3})
  MaterialFunctionCall(
   UVs=@n18
   Texture Scale=MaterialFunctionCall(X=ConstantBiasScale(None=@n20), Y=Constant() {r=1}) {materialFunction=&MakeFloat2}.Result
  ) {materialFunction=&ScaleUVsByCenter}
  MaterialFunctionCall!(NumberOfMips) {materialFunction=&/MF_LemonCart_CustomScreenMIPOut}
 ]}
