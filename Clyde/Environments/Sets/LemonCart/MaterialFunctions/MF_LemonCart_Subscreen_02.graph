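# Material function for a LemonCart subscreen, outputting color and alpha for an animated display graphic.
# A stretched radial gradient masks a dark backdrop with dim sine-shaped rings, faded by a vertical gradient.
# The absolute difference of two time-rotated angular gradients, offset in phase, modulates the rings as a sweep, tuned by Contrast and Multi parameters.
# A flipbook texture driven by the Speed input composites on top through its alpha times Mask, which also extends the output alpha.
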MaterialFunction /Clyde/Environments/Sets/LemonCart/MaterialFunctions/MF_LemonCart_Subscreen_02
graph {nodes [
  n1 = FunctionInput(Preview=TextureCoordinate()) {bUsePreviewValueAsDefault=true, inputName=UVs, inputType=FunctionInput_Vector2}
  n2 = Add(A=Multiply(A=@n1, B=Constant2Vector() {g=4, r=1}), B=Constant2Vector() {g=-2.3})
  n3 = MaterialFunctionCall(UVs=@n2, Radius=Constant() {r=0.48}, Density=Constant() {r=100}) {materialFunction=&RadialGradientExponential}
  n4 = LinearInterpolate(B=Constant3Vector() {constant=<rgba 0.0077 0.0096 0.01 1>}, Alpha=@n3.RadialGradientExponential)
  n5 = Constant() {r=2}
  n6 = Sine(None=Multiply(A=MaterialFunctionCall(UVs=@n2, Density=@n5) {materialFunction=&RadialGradientExponential}.RadialGradientExponential) {constB=10})
  n7 = Constant() {r=4}
  n8 = Sine(None=Multiply(A=MaterialFunctionCall(UVs=@n2, Density=@n7) {materialFunction=&RadialGradientExponential}.RadialGradientExponential) {constB=12}) {period=0.6}
  n9 = Multiply(
   A=LinearInterpolate(A=Saturate(None=@n6), B=Saturate(None=@n8)) {constAlpha=0.2}
   B=Power(Base=MaterialFunctionCall() {materialFunction=&LinearGradient}.VGradient)
  )
  n10 = Constant()
  n11 = Add(
   A=TextureCoordinate()
   B=MaterialFunctionCall(Y=Constant() {r=-0.2}) {materialFunction=&MakeFloat2}.Result
  )
  n12 = Time()
  n13 = Add(A=@n12)
  n14 = Add(A=MaterialFunctionCall(UVs=@n11, Rotation Center=Constant() {r=0.5}, "Rotation Angle (0-1)"=Multiply(A=@n13) {constB=0.2}) {materialFunction=&CustomRotator}.Rotated Values) {constB=-0.5}
  n15 = ScalarParameter() {defaultValue=8, parameterName=Subscreen02.Contrast}
  n16 = Power(
   Base=MaterialFunctionCall(
    A=@n10
    B=Constant() {r=1}
    C=@n10
    Alpha=MaterialFunctionCall(Vector or UVs=@n14) {materialFunction=&VectorToRadialValue}.Vector Converted to Angle
   ) {materialFunction=&Lerp_3Color}.Result
   Exp=@n15
  ) {constExponent=8}
  n17 = Constant()
  n18 = Add(A=MaterialFunctionCall(UVs=@n11, Rotation Center=Constant() {r=0.5}, "Rotation Angle (0-1)"=Multiply(A=@n12) {constB=0.2}) {materialFunction=&CustomRotator}.Rotated Values) {constB=-0.5}
  n19 = Power(
   Base=MaterialFunctionCall(
    A=@n17
    B=Constant() {r=1}
    C=@n17
    Alpha=MaterialFunctionCall(Vector or UVs=@n18) {materialFunction=&VectorToRadialValue}.Vector Converted to Angle
   ) {materialFunction=&Lerp_3Color}.Result
   Exp=@n15
  ) {constExponent=8}
  n20 = Max(A=Subtract(A=@n16, B=@n19), B=Subtract(A=@n19, B=@n16))
  n21 = ScalarParameter() {defaultValue=1, parameterName=Subscreen02.Multi}
  n22 = Multiply(A=@n9, B=Saturate(None=Multiply(A=@n20, B=@n21)))
  n23 = Multiply(A=LinearInterpolate(B=LinearInterpolate(A=@n9, B=@n22) {constAlpha=0.9}, Alpha=@n3.RadialGradientExponential)) {constB=0.2}
  n24 = Constant() {r=1}
  n25 = MaterialFunctionCall(
   "Animation  Phase (0-1)"=Multiply(A=Time(), B=FunctionInput(Preview=@n24) {bUsePreviewValueAsDefault=true, inputName=Speed, inputType=FunctionInput_Scalar, sortPriority=1})
   Number of Rows=Constant() {r=2}
   Number of Columns=Constant() {r=4}
   Texture=TextureObject() {texture=/Clyde/Environments/Sets/LemonCart/Textures/T_LemonCart_Screen_02}
   UVs=@n1
  ) {materialFunction=&FlipBook}
  n26 = Multiply(
   A=@n25.Alpha
   B=FunctionInput(Preview=Constant() {r=1}) {bUsePreviewValueAsDefault=true, inputName=Mask, inputType=FunctionInput_Scalar, sortPriority=3}
  )
  FunctionOutput(None=LinearInterpolate(A=Add(A=@n4, B=@n23), B=@n25.Result, Alpha=@n26))
  FunctionOutput(None=Max(A=@n3.RadialGradientExponential, B=@n26)) {outputName=Alpha, sortPriority=1}
  TextureCoordinate()
 ]}
