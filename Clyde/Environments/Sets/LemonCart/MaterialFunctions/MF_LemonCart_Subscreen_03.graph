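# Material function for the LemonCart's third sub-screen, outputting an icon's animated two-color fill and alpha mask.
# Color blends SubScreen03 ColorA and ColorB through a time-scrolled, slightly rotated gradient band that forms a looping shimmer.
# Alpha takes the screen texture's green channel, using center-scaled UVs with height-driven BumpOffset parallax and jitter input, clipped to the 0-1 UV range.
# A sine-driven UV squash and a custom MIP call are left unconnected.
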MaterialFunction /Clyde/Environments/Sets/LemonCart/MaterialFunctions/MF_LemonCart_Subscreen_03
graph {nodes [
  n1 = VectorParameter() {parameterName=SubScreen03.ColorA, defaultValue=<rgba 1 0.8013 0 1>}
  n2 = VectorParameter() {parameterName=SubScreen03.ColorB, defaultValue=<rgba 1 0.2233 0 1>}
  n3 = TextureCoordinate()
  n4 = Constant() {r=0.1}
  n5 = Frac(None=Multiply(A=Time(), B=ScalarParameter() {defaultValue=0.33, parameterName=SubScreen03.ShimmerSpeed}) {constB=0.5})
  n6 = Multiply(A=Add(
    A=MaterialFunctionCall(UVs=@n3, "Rotation Angle (0-1)"=@n4) {materialFunction=&CustomRotator}.Rotated Values
    B=LinearInterpolate(Alpha=@n5) {constA=1, constB=-1}
   ) {constB=0}) {constB=3}
  n7 = MaterialFunctionCall(UV Channel=@n6) {materialFunction=&LinearGradient}
  n8 = Multiply(A=Multiply(
    A=MaterialFunctionCall(UV Channel=@n6) {materialFunction=&LinearGradient}.UGradient
    B=OneMinus(None=@n7.UGradient)
   )) {constB=5}
  n9 = LinearInterpolate(B=Power(Base=Saturate(None=@n8)), Alpha=Saturate(None=ScalarParameter() {defaultValue=1, parameterName=SubScreen03.ColorB.Blend}))
  FunctionOutput(None=Saturate(None=LinearInterpolate(A=@n1.RGB, B=@n2.RGB, Alpha=@n9)))
  n10 = FunctionInput() {bUsePreviewValueAsDefault=true, inputName=Height Bump Offset, inputType=FunctionInput_Scalar, sortPriority=1}
  n11 = BumpOffset(
   Coordinate=FunctionInput(Preview=TextureCoordinate()) {bUsePreviewValueAsDefault=true, inputName=UVs, inputType=FunctionInput_Vector2}
   Height=@n10
  )
  n12 = ScalarParameter() {defaultValue=1, parameterName=SubScreen03.Icon.Scale}
  n13 = Add(
   A=Add(
    A=MaterialFunctionCall(UVs=@n11, Texture Scale=@n12) {materialFunction=&ScaleUVsByCenter}.UVs
    B=FunctionInput() {bUsePreviewValueAsDefault=true, inputName=Vertical Jitter, inputType=FunctionInput_Vector2, sortPriority=2}
   )
   B=MaterialFunctionCall(X=Multiply(A=@n10) {constB=-0.1}, Y=Constant()) {materialFunction=&MakeFloat2}.Result
  )
  n14 = TextureSample(UVs=@n13) {texture=/Clyde/Environments/Sets/LemonCart/Textures/T_LemonCart_Screen_01}
  n15 = MaterialFunctionCall(UVs=@n11, Texture Scale=Multiply(A=ScalarParameter() {defaultValue=1, parameterName=SubScreen03.Icon.Scale}) {constB=0.98}) {materialFunction=&ScaleUVsByCenter}
  FunctionOutput(None=Saturate(None=Multiply(A=@n14.G, B=@n15.0-1 mask))) {outputName=Alpha, sortPriority=1}
  n16 = Sine(None=Multiply(A=Time()) {constB=0.3})
  MaterialFunctionCall(
   UVs=@n13
   Texture Scale=MaterialFunctionCall(X=ConstantBiasScale(None=@n16), Y=Constant() {r=1}) {materialFunction=&MakeFloat2}.Result
  ) {materialFunction=&ScaleUVsByCenter}
  MaterialFunctionCall!(NumberOfMips) {materialFunction=&/MF_LemonCart_CustomScreenMIPOut}
 ]}
