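# Material function that plays an animated subscreen graphic for the LemonCart set's screens.
# A FlipBook cycles through a 4-row, 5-column sheet from T_LemonCart_Screen_03, with its phase driven by Time at half rate times the Speed input.
# An optional UVs input sets placement, defaulting to the mesh texture coordinates.
# It outputs the frame result, plus an Alpha output of the frame's alpha multiplied by the Mask input.
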MaterialFunction /Clyde/Environments/Sets/LemonCart/MaterialFunctions/MF_LemonCart_Subscreen_04
graph {nodes [
  n1 = Time()
  n2 = Constant() {r=1}
  n3 = MaterialFunctionCall(
   "Animation  Phase (0-1)"=Multiply(
    A=Multiply(A=@n1) {constB=0.5}
    B=FunctionInput(Preview=@n2) {bUsePreviewValueAsDefault=true, inputName=Speed, inputType=FunctionInput_Scalar, sortPriority=1}
   )
   Number of Rows=Constant() {r=4}
   Number of Columns=Constant() {r=5}
   Texture=TextureObject() {texture=/Clyde/Environments/Sets/LemonCart/Textures/T_LemonCart_Screen_03}
   UVs=FunctionInput(Preview=TextureCoordinate()) {bUsePreviewValueAsDefault=true, inputName=UVs, inputType=FunctionInput_Vector2}
  ) {materialFunction=&FlipBook}
  FunctionOutput(None=@n3.Result)
  n4 = Constant() {r=1}
  FunctionOutput(None=Multiply(A=@n3.Alpha, B=FunctionInput(Preview=@n4) {bUsePreviewValueAsDefault=true, inputName=Mask, inputType=FunctionInput_Scalar, sortPriority=3})) {outputName=Alpha, sortPriority=1}
 ]}
