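# Material function that renders a looping animated screen transition for the LemonCart set, with horizontal and vertical swipes and a composited logo.
# Offset time wraps over hold plus transition length and samples a curve atlas, driving each swipe's position.
# Scrolling textures, gradients, Perlin distortion and a radial sine wave build masks that blend three color parameters and the logo.
# Ray-traced and material-proxy paths get a simpler color blend without the logo.
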MaterialFunction /Clyde/Environments/Sets/LemonCart/MaterialFunctions/MF_LemonCart_Transition
graph {nodes [
  n1 = FunctionInput(Preview=Constant4Vector().RGBA) {bUsePreviewValueAsDefault=true, inputName=Logo RGBA, inputType=FunctionInput_Vector4, sortPriority=5}
  n2 = VectorParameter() {parameterName=TransitionScreen.ColorC, defaultValue=<rgba 0.3333 0.6233 1 1>}
  n3 = VectorParameter() {parameterName=TransitionScreen.ColorA, defaultValue=<rgba 0 0.0123 0.0573 1>}
  n4 = LinearInterpolate(
   A=LinearInterpolate(
    A=@n3.RGB
    B=VectorParameter() {parameterName=TransitionScreen.ColorB, defaultValue=<rgba 0.0048 0.1046 0.4342 1>}.RGB
    Alpha=NamedRerouteUsage()
   )
   B=LinearInterpolate(A=@n3.RGB, B=@n2.RGB, Alpha=NamedRerouteUsage())
   Alpha=NamedRerouteUsage()
  )
  n5 = LinearInterpolate(A=@n4, B=ComponentMask(None=@n1) {b=true, g=true, r=true}, Alpha=ComponentMask(None=@n1) {a=true})
  n6 = NamedRerouteUsage()
  n7 = LinearInterpolate(
   A=LinearInterpolate(A=@n4, B=LinearInterpolate(A=@n4, B=@n5, Alpha=@n6), Alpha=NamedRerouteUsage())
   B=LinearInterpolate(A=@n3.RGB, B=@n2.RGB, Alpha=NamedRerouteUsage())
   Alpha=NamedRerouteUsage()
  )
  FunctionOutput(None=MaterialProxyReplace(Realtime=RayTracingQualitySwitch(Normal=@n7, RayTraced=@n4), MaterialProxy=@n4))
  n8 = NamedRerouteUsage()
  n9 = Constant() {r=0.75}
  n10 = MaterialFunctionCall(X=Constant() {r=2}, Y=@n9) {materialFunction=&MakeFloat2}
  n11 = VectorParameter() {parameterName=ScreenTransition.Bounds, channelNames {a=VSwipeMax, b=VSwipeMin, g=HSwipeMax, r=HSwipeMin}, defaultValue=<rgba -1.2 1 -1 1>}
  n12 = Add(A=NamedRerouteUsage(), B=Divide(A=NamedRerouteUsage()) {constB=-16}) {constB=-0.5}
  n13 = NamedRerouteUsage()
  n14 = CurveAtlasRowParameter(CurveTime=Divide(A=Modulo!(A=@n12, B=@n13), B=@n13)) {
   atlas=/Clyde/Environments/Sets/ClayPlug/TechArt/CA_Clyde
   curve=/Clyde/Environments/Sets/LemonCart/Textures/CV_LemonCart_ScreenTransitionTime
   defaultValue=1
   parameterName=ScreenTransition.TimingCurve
  }
  n15 = MaterialFunctionCall(X=Constant(), Y=LinearInterpolate(A=@n11.VSwipeMin, B=@n11.VSwipeMin, Alpha=@n14.B) {constA=-1}) {materialFunction=&MakeFloat2}
  n16 = Subtract(A=Multiply(A=@n8, B=@n10.Result), B=@n15.Result)
  n17 = Add(A=@n16, B=Multiply(A=NamedRerouteUsage()) {constB=0.1})
  n18 = Add(A=@n17, B=Constant2Vector() {g=-0.05})
  n19 = Multiply(A=Power(Base=MaterialFunctionCall(UV Channel=@n18) {materialFunction=&LinearGradient}.VGradient)) {constB=5}
  n20 = Multiply(
   A=@n8
   B=MaterialFunctionCall(X=Constant() {r=1}, Y=@n9) {materialFunction=&MakeFloat2}.Result
  )
  n21 = Constant() {r=3}
  n22 = Constant() {r=0.95}
  n23 = Multiply(
   A=Multiply(A=TextureSample(UVs=@n16) {texture=/Clyde/Environments/Sets/LemonCart/Textures/T_LemonCart_SubScreenTransition}.R, B=Saturate(None=@n19))
   B=MaterialFunctionCall(
    UVs=Subtract(A=@n20, B=@n15.Result)
    Texture Scale=MaterialFunctionCall(X=@n21, Y=@n22) {materialFunction=&MakeFloat2}.Result
   ) {materialFunction=&ScaleUVsByCenter}.0-1 mask
  )
  NamedRerouteDeclaration(None=Saturate(None=Multiply(A=@n23) {constB=3})) {name=VSwipeMask, nodeColor=<rgba 1 0 0.1641 1>}
  n24 = NamedRerouteUsage()
  n25 = Multiply(A=Time()) {constB=-0.05}
  n26 = OneMinus(None=TextureSample(UVs=Add(A=@n24, B=@n25)) {texture=&T_LemonCart_SubScreenTransition}.B)
  n27 = VectorParameter() {parameterName=ScreenTransition.Bounds, channelNames {a=VSwipeMax, b=VSwipeMin, g=HSwipeMax, r=HSwipeMin}, defaultValue=<rgba -1.2 1 -1 1>}
  n28 = NamedRerouteUsage()
  n29 = Divide(A=Modulo(A=NamedRerouteUsage(), B=@n28), B=@n28)
  n30 = CurveAtlasRowParameter(CurveTime=@n29) {
   atlas=&CA_Clyde
   curve=&CV_LemonCart_ScreenTransitionTime
   defaultValue=1
   parameterName=ScreenTransition.TimingCurve
  }
  n31 = LinearInterpolate(A=@n27.HSwipeMin, B=@n27.HSwipeMin, Alpha=@n30.R) {constA=-1.2}
  n32 = NamedRerouteUsage()
  n33 = Multiply(A=@n29) {constB=0.1}
  n34 = NamedRerouteDeclaration(None=TextureSample(UVs=Add(A=@n32, B=@n33)) {samplerType=SAMPLERTYPE_LinearColor, texture=/Game/Effects/SnowSandInteraction/Textures/SSI_Perlin}.R) {name=PerlinNoise, nodeColor=<rgba 0 1 0.3906 1>}
  n35 = MaterialFunctionCall(UV Channel=Add(A=Subtract(A=@n24, B=@n31), B=Multiply(A=@n34) {constB=0.1})) {materialFunction=&LinearGradient}
  n36 = Constant()
  n37 = LinearInterpolate(Alpha=Saturate(None=If(A=@n35.UGradient, "A > B"=@n36, "A == B"=@n35.UGradient, "A < B"=@n35.UGradient) {constB=1})) {constA=0.2}
  n38 = VectorParameter() {parameterName=ScreenTransiton.RadialWave, channelNames {a=Speed, b=Radius, g=PosY, r=PosX}, defaultValue=<rgba -0.25 0.25 3 0.8>}
  n39 = MaterialFunctionCall(
   UVs=NamedRerouteUsage()
   CenterPosition=MaterialFunctionCall(X=@n38.PosX, Y=@n38.PosX) {materialFunction=&MakeFloat2}.Result
   Radius=@n38.Radius
  ) {materialFunction=&RadialGradientExponential}
  n40 = Time()
  n41 = Sine(None=Add(A=Multiply(A=@n39.RadialGradientExponential) {constB=4}, B=Multiply(A=@n40, B=@n38.Speed) {constB=0.8}))
  n42 = Multiply(A=Multiply(A=@n39.RadialGradientExponential, B=ConstantBiasScale(None=@n41)))
  NamedRerouteDeclaration(None=Add(A=Multiply(A=@n26, B=@n37), B=Multiply(A=@n42, B=@n37))) {name=HSwipeGreyscale, nodeColor=<rgba 0 0.4844 1 1>}
  NamedRerouteDeclaration(None=FunctionInput(Preview=TextureCoordinate()) {bUsePreviewValueAsDefault=true, inputName=UVs, inputType=FunctionInput_Vector2}) {name=UVs, nodeColor=<rgba 1 0 0.6094 1>}
  n43 = NamedRerouteUsage()
  n44 = Constant() {r=0.75}
  n45 = Multiply(
   A=@n43
   B=MaterialFunctionCall(X=Constant() {r=2.33}, Y=@n44) {materialFunction=&MakeFloat2}.Result
  )
  n46 = VectorParameter() {parameterName=ScreenTransition.Bounds, channelNames {a=VSwipeMax, b=VSwipeMin, g=HSwipeMax, r=HSwipeMin}, defaultValue=<rgba -1.2 1 -1 1>}
  n47 = NamedRerouteUsage()
  n48 = Divide(A=@n47) {constB=16}
  n49 = Add(A=Add(A=NamedRerouteUsage(), B=Divide(A=@n47) {constB=-16}) {constB=-0.5}, B=Multiply(A=@n48) {constB=2}) {constB=-0.5}
  n50 = Add(A=@n49, B=@n48)
  n51 = NamedRerouteUsage()
  n52 = CurveAtlasRowParameter(CurveTime=Divide(A=Modulo(A=@n50, B=@n51), B=@n51)) {
   atlas=&CA_Clyde
   curve=&CV_LemonCart_ScreenTransitionTime
   defaultValue=1
   parameterName=ScreenTransition.TimingCurve
  }
  n53 = MaterialFunctionCall(X=Constant(), Y=LinearInterpolate(A=@n46.VSwipeMin, B=@n46.VSwipeMin, Alpha=@n52.G) {constA=-1}) {materialFunction=&MakeFloat2}
  n54 = MaterialFunctionCall(
   UVs=Subtract(A=Subtract(A=@n45, B=@n53.Result), B=Constant2Vector() {g=-0.25})
   Texture Scale=MaterialFunctionCall(X=Constant() {r=2}, Y=Constant() {r=0.95}) {materialFunction=&MakeFloat2}.Result
  ) {materialFunction=&ScaleUVsByCenter}
  n55 = NamedRerouteUsage()
  n56 = CurveAtlasRowParameter(CurveTime=Divide(A=Modulo(A=@n49, B=@n55), B=@n55)) {
   atlas=&CA_Clyde
   curve=&CV_LemonCart_ScreenTransitionTime
   defaultValue=1
   parameterName=ScreenTransition.TimingCurve
  }
  n57 = MaterialFunctionCall(X=Constant(), Y=LinearInterpolate(A=@n46.VSwipeMin, B=@n46.VSwipeMin, Alpha=@n56.G) {constA=-1}) {materialFunction=&MakeFloat2}
  n58 = Subtract(A=@n45, B=@n57.Result)
  n59 = Add(
   A=Multiply(
    A=TextureSample(UVs=@n54.UVs) {desc=Secondary lines moving at different speed, texture=&T_LemonCart_SubScreenTransition}.G
    B=@n54.0-1 mask
   )
   B=TextureSample(UVs=@n58) {texture=&T_LemonCart_SubScreenTransition}.R
  )
  n60 = NamedRerouteUsage()
  n61 = Add(A=Add(A=@n58, B=Multiply(A=@n60) {constB=0.17}), B=Constant2Vector() {g=-0.05})
  n62 = Multiply(A=Power(Base=MaterialFunctionCall(UV Channel=@n61) {materialFunction=&LinearGradient}.VGradient) {constExponent=2.5}) {constB=5}
  n63 = Multiply(
   A=@n43
   B=MaterialFunctionCall(X=Constant() {r=1}, Y=@n44) {materialFunction=&MakeFloat2}.Result
  )
  n64 = Constant() {r=3}
  n65 = Constant() {r=0.95}
  n66 = Multiply(
   A=Multiply(A=Saturate(None=@n59), B=Saturate(None=@n62))
   B=MaterialFunctionCall(
    UVs=Subtract(A=@n63, B=@n57.Result)
    Texture Scale=MaterialFunctionCall(X=@n64, Y=@n65) {materialFunction=&MakeFloat2}.Result
   ) {materialFunction=&ScaleUVsByCenter}.0-1 mask
  )
  NamedRerouteDeclaration(None=Saturate(None=Multiply(A=@n66) {constB=3})) {name=VSwipeGreyscale_BG_Mask, nodeColor=<rgba 0.8047 1 0 1>}
  NamedRerouteDeclaration(None=Power(Base=@n66) {constExponent=4}) {name=VSwipeGreyscale_BG, nodeColor=<rgba 0.8047 1 0 1>}
  n67 = Subtract(A=Multiply(A=NamedRerouteUsage(), B=@n10.Result), B=@n15.Result)
  n68 = OneMinus(None=MaterialFunctionCall(UV Channel=Add(A=@n67) {constB=0.5}) {materialFunction=&LinearGradient}.VGradient)
  n69 = Constant() {r=0.5}
  n70 = Constant() {r=-0.5}
  n71 = Multiply(
   A=Add(
    A=MaterialFunctionCall(UVs=@n67, "Rotation Angle (0-1)"=@n69) {materialFunction=&CustomRotator}.Rotated Values
    B=MaterialFunctionCall(Y=@n70) {materialFunction=&MakeFloat2}.Result
   )
   B=MaterialFunctionCall(X=Constant() {r=1}, Y=Constant() {r=2}) {materialFunction=&MakeFloat2}.Result
  )
  n72 = MaterialFunctionCall(X=Constant() {r=4}, Y=Constant() {r=0.95}) {materialFunction=&MakeFloat2}
  n73 = Multiply(
   A=Multiply(A=Power(Base=@n68)) {constB=10}
   B=Multiply(
    A=TextureSample(UVs=@n71) {texture=&T_LemonCart_SubScreenTransition}.G
    B=MaterialFunctionCall(UVs=@n71, Texture Scale=@n72.Result) {materialFunction=&ScaleUVsByCenter}.0-1 mask
   )
  )
  NamedRerouteDeclaration(None=Saturate(None=Add(A=@n23, B=@n73))) {name=VSwipeGreyscale, nodeColor=<rgba 0.8047 1 0 1>}
  NamedRerouteDeclaration(None=@n37) {name=HSwipeMaskFaded, nodeColor=<rgba 0 0.0859 1 1>}
  n74 = Constant2Vector() {g=0.5}
  n75 = Power(Base=MaterialFunctionCall(UV Channel=Add(A=@n17, B=@n74)) {materialFunction=&LinearGradient}.VGradient) {constExponent=50}
  NamedRerouteDeclaration(None=Saturate(None=Multiply(A=@n75))) {name=VGradientMask, nodeColor=<rgba 0.3125 0 1 1>}
  n76 = Constant() {r=1}
  NamedRerouteDeclaration(None=If(A=@n35.UGradient, "A > B"=@n36, "A == B"=@n76, "A < B"=@n76) {constB=1}) {name=HSwipeMaskRounded, nodeColor=<rgba 0 0.0859 1 1>}
  n77 = FunctionInput(Preview=Constant()) {bUsePreviewValueAsDefault=true, inputName=TimeOffset, inputType=FunctionInput_Scalar, sortPriority=5}
  n78 = Time()
  n79 = FunctionInput(Preview=Constant()) {bUsePreviewValueAsDefault=true, inputName=HoldLength, inputType=FunctionInput_Scalar, sortPriority=4}
  NamedRerouteDeclaration(None=Add(A=Add(A=@n77, B=@n78), B=Divide(A=@n79))) {name=In.TimeWithOffset, nodeColor=<rgba 1 0 0.8203 1>}
  n80 = FunctionInput(Preview=Constant() {r=8}) {bUsePreviewValueAsDefault=true, inputName=TransitionLength, inputType=FunctionInput_Scalar, sortPriority=3}
  NamedRerouteDeclaration(None=Add(A=@n79, B=NamedRerouteDeclaration(None=@n80) {name=TransitionLength, nodeColor=<rgba 1 0.2812 0 1>})) {name=TotalTime, nodeColor=<rgba 0.2656 1 0 1>}
 ]}
